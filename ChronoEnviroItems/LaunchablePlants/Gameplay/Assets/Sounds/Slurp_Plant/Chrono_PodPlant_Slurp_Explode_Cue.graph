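# Sound cue for the explosion of a launchable Slurp pod plant, using a Slurp explode attenuation asset and the medium explosives sound class.
# A mixer layers three branches: a randomized close explosion plus two heal-sound layers that share the same four waves.
# One heal layer is pitched up and slightly delayed, the other pitched down.
# A quality-level node switches the explosion between an eight-variation pool and a three-variation subset.
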SoundCue /ChronoEnviroItems/LaunchablePlants/Gameplay/Assets/Sounds/Slurp_Plant/Chrono_PodPlant_Slurp_Explode_Cue
root=Mixer_1
nodes [
 {id=Delay_0, children [Modulator_2], properties {delayMax=0.1, delayMin=0.1}}
 {id=Mixer_1, children [Modulator_0,Delay_0,Modulator_1], properties {inputVolume [0.5,4,3]}}
 {id=Modulator_0, children [QualityLevel_0]}
 {id=Modulator_1, children [Random_1], properties {pitchMax=0.85, pitchMin=0.75}}
 {id=Modulator_2, children [Random_3], properties {pitchMax=1.35, pitchMin=1.2}}
 {id=QualityLevel_0, children [Random_0,Random_0,Random_2,Random_2]}
 {
  id=Random_0
  children [WavePlayer_2,WavePlayer_3,WavePlayer_4,WavePlayer_5,WavePlayer_6,WavePlayer_7,WavePlayer_8,WavePlayer_9]
  properties {weights [1,1,1,1,1,1,1,1]}
 }
 {id=Random_1, children [WavePlayer_10, WavePlayer_11, WavePlayer_12, WavePlayer_13], properties {weights [1,1,1,1]}}
 {id=Random_2, children [WavePlayer_7,WavePlayer_8,WavePlayer_9], properties {bShouldExcludeFromBranchCulling=true, weights [1,1,1]}}
 {id=Random_3, children [WavePlayer_0, WavePlayer_1, WavePlayer_18, WavePlayer_19], properties {weights [1,1,1,1]}}
 {
  id=WavePlayer_0
  properties {soundWaveAssetPtr=&/PodPlant_Slurp_Explode_Heal_04}
 }
 {
  id=WavePlayer_1
  properties {soundWaveAssetPtr=&/PodPlant_Slurp_Explode_Heal_01}
 }
 {
  id=WavePlayer_10
  properties {soundWaveAssetPtr=&PodPlant_Slurp_Explode_Heal_04}
 }
 {
  id=WavePlayer_11
  properties {soundWaveAssetPtr=&PodPlant_Slurp_Explode_Heal_01}
 }
 {
  id=WavePlayer_12
  properties {soundWaveAssetPtr=&/PodPlant_Slurp_Explode_Heal_02}
 }
 {
  id=WavePlayer_13
  properties {soundWaveAssetPtr=&/PodPlant_Slurp_Explode_Heal_03}
 }
 {
  id=WavePlayer_18
  properties {soundWaveAssetPtr=&PodPlant_Slurp_Explode_Heal_02}
 }
 {
  id=WavePlayer_19
  properties {soundWaveAssetPtr=&PodPlant_Slurp_Explode_Heal_03}
 }
 {
  id=WavePlayer_2
  properties {soundWaveAssetPtr=&/PodPlant_PodExplode_Slurp_Close_01}
 }
 {
  id=WavePlayer_3
  properties {soundWaveAssetPtr=&/PodPlant_PodExplode_Slurp_Close_02}
 }
 {
  id=WavePlayer_4
  properties {soundWaveAssetPtr=&/PodPlant_PodExplode_Slurp_Close_03}
 }
 {
  id=WavePlayer_5
  properties {soundWaveAssetPtr=&/PodPlant_PodExplode_Slurp_Close_04}
 }
 {
  id=WavePlayer_6
  properties {soundWaveAssetPtr=&/PodPlant_PodExplode_Slurp_Close_05}
 }
 {
  id=WavePlayer_7
  properties {soundWaveAssetPtr=&/PodPlant_PodExplode_Slurp_Close_06}
 }
 {
  id=WavePlayer_8
  properties {soundWaveAssetPtr=&/PodPlant_PodExplode_Slurp_Close_07}
 }
 {
  id=WavePlayer_9
  properties {soundWaveAssetPtr=&/PodPlant_PodExplode_Slurp_Close_08}
 }
]
properties {
 attenuationSettings=/ChronoEnviroItems/LaunchablePlants/Gameplay/Assets/Sounds/Launchable_Plants_Slurp_Explode_Att
 maxDistance=3512
 soundClassObject=/Game/Sounds/FortSoundClassesAndModes/Fort_Explosives_Medium_PSM
 concurrencyOverrides {maxCount=4, resolutionRule=PreventNew}
 concurrencySet [&RifleBulletImpact_Conc]
}
