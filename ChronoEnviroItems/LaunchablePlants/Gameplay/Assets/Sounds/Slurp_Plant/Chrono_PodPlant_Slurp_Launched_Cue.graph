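# Sound cue for the launch of a Slurp pod plant, mixing a launch layer with a delayed slurp sweetener.
# The launch layer pairs a random Stink plant pod-launch sample with one of two briefly delayed tree-leaf rustles.
# The sweetener is pitched down and boosted, with a quality-level node choosing between two overlapping random impact pools.
# Shared launched-plant attenuation and a concurrency limit that blocks new instances govern playback.
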SoundCue /ChronoEnviroItems/LaunchablePlants/Gameplay/Assets/Sounds/Slurp_Plant/Chrono_PodPlant_Slurp_Launched_Cue
root=Mixer_0
nodes [
 {id=Delay_2, children [Modulator_1], properties {delayMax=0.25, delayMin=0.25}}
 {id=Delay_3, children [WavePlayer_4], properties {delayMax=0.15, delayMin=0.15}}
 {id=Delay_4, children [WavePlayer_5], properties {delayMax=0.2, delayMin=0.2}}
 {id=Mixer_0, children [Modulator_0,Delay_2], properties {inputVolume [1,1.8]}}
 {id=Mixer_1, children [Random_5,Random_6], properties {inputVolume [0.6,2]}}
 {id=Modulator_0, children [Mixer_1]}
 {id=Modulator_1, children [QualityLevel_1], properties {pitchMax=0.8, pitchMin=0.8, volumeMax=2, volumeMin=2}}
 {id=QualityLevel_1, children [Random_1,Random_1,Random_3,Random_3]}
 {id=Random_1, children [WavePlayer_0, WavePlayer_1, WavePlayer_10, WavePlayer_11], properties {weights [1,1,1,1]}}
 {id=Random_3, children [WavePlayer_1, WavePlayer_10, WavePlayer_11], properties {bShouldExcludeFromBranchCulling=true, weights [1,1,1]}}
 {id=Random_5, children [Delay_3,Delay_4], properties {weights [1,1]}}
 {id=Random_6, children [WavePlayer_6,WavePlayer_7,WavePlayer_9], properties {bShouldExcludeFromBranchCulling=true, weights [1,1,1]}}
 {
  id=WavePlayer_0
  properties {soundWaveAssetPtr=&/PodPlant_Impact_Slurp_Sweetner_01}
 }
 {
  id=WavePlayer_1
  properties {soundWaveAssetPtr=&/PodPlant_Impact_Slurp_Sweetner_02}
 }
 {
  id=WavePlayer_10
  properties {soundWaveAssetPtr=&/PodPlant_Impact_Slurp_Sweetner_03}
 }
 {
  id=WavePlayer_11
  properties {soundWaveAssetPtr=&/PodPlant_Impact_Slurp_Sweetner_04}
 }
 {id=WavePlayer_4, properties {soundWaveAssetPtr=/Game/Sounds/Fort_Impact_Sounds/Demolition/Tree/Demo_Tree_Default_Destroyed_Leaves_04}}
 {id=WavePlayer_5, properties {soundWaveAssetPtr=/Game/Sounds/Fort_Impact_Sounds/Demolition/Tree/Demo_Tree_Default_Destroyed_Leaves_03}}
 {
  id=WavePlayer_6
  properties {soundWaveAssetPtr=/ChronoEnviroItems/LaunchablePlants/Gameplay/Assets/Sounds/Stink_Plant/PodPlant_PodLaunch_Stink_01}
 }
 {
  id=WavePlayer_7
  properties {soundWaveAssetPtr=/ChronoEnviroItems/LaunchablePlants/Gameplay/Assets/Sounds/Stink_Plant/PodPlant_PodLaunch_Stink_02}
 }
 {
  id=WavePlayer_9
  properties {soundWaveAssetPtr=/ChronoEnviroItems/LaunchablePlants/Gameplay/Assets/Sounds/Stink_Plant/PodPlant_PodLaunch_Stink_08}
 }
]
properties {
 attenuationSettings=/ChronoEnviroItems/LaunchablePlants/Gameplay/Assets/Sounds/Launchable_Plants_Launched_Att
 maxDistance=5008.436
 soundClassObject=&Fort_Impacts
 volumeMultiplier=3
 concurrencyOverrides {maxCount=4, resolutionRule=PreventNew}
 concurrencySet [&RifleBulletImpact_Conc]
}
