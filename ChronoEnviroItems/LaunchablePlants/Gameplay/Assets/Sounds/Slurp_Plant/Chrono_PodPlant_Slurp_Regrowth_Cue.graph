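# Sound cue for the regrowth sound of the launchable Slurp pod plant, classed as Athena ambient audio.
# A Random node picks one of four equally weighted, shared PodPlant regrowth waves, which a Modulator boosts in volume at fixed pitch.
# It reuses the launchable plants' ambient loop attenuation and overrides concurrency to cap simultaneous instances at a small count.
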SoundCue /ChronoEnviroItems/LaunchablePlants/Gameplay/Assets/Sounds/Slurp_Plant/Chrono_PodPlant_Slurp_Regrowth_Cue
root=Modulator_1
nodes [
 {id=Modulator_1, children [Random_0], properties {pitchMax=1, pitchMin=1, volumeMax=2, volumeMin=2}}
 {
  id=Random_0
  children [WavePlayer_0,WavePlayer_1,WavePlayer_2,WavePlayer_3]
  properties {bShouldExcludeFromBranchCulling=true, weights [1,1,1,1]}
 }
 {id=WavePlayer_0, properties {soundWaveAssetPtr=/ChronoEnviroItems/LaunchablePlants/Gameplay/Assets/Sounds/Chrono_PodPlant_Regrowth_01}}
 {id=WavePlayer_1, properties {soundWaveAssetPtr=/ChronoEnviroItems/LaunchablePlants/Gameplay/Assets/Sounds/Chrono_PodPlant_Regrowth_02}}
 {id=WavePlayer_2, properties {soundWaveAssetPtr=/ChronoEnviroItems/LaunchablePlants/Gameplay/Assets/Sounds/Chrono_PodPlant_Regrowth_03}}
 {id=WavePlayer_3, properties {soundWaveAssetPtr=/ChronoEnviroItems/LaunchablePlants/Gameplay/Assets/Sounds/Chrono_PodPlant_Regrowth_04}}
]
properties {
 attenuationSettings=/ChronoEnviroItems/LaunchablePlants/Gameplay/Assets/Sounds/Launchable_Plants_Ambient_Loop_Att
 bOverrideConcurrency=true
 maxDistance=1750
 soundClassObject=&Fort_Ambient_Athena
 volumeMultiplier=1.504
 concurrencyOverrides {maxCount=3}
 concurrencySet [null]
}
